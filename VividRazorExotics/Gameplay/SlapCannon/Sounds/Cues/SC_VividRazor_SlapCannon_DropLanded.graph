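# Sound cue for the VividRazor Slap Cannon weapon landing after being dropped.
# A Random node picks one of three equally weighted StickyGrenadeLauncher drop-land waves and feeds a root Modulator left at default settings.
# It plays at reduced volume through weapon-drop attenuation and the loot-collect sound class, and uses the shared Weapon_Drop_StopOldest concurrency set.
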SoundCue /VividRazorExotics/Gameplay/SlapCannon/Sounds/Cues/SC_VividRazor_SlapCannon_DropLanded
root=Modulator_0
nodes [
 {id=Modulator_0, children [Random_0]}
 {id=Random_0, children [WavePlayer_3,WavePlayer_4,WavePlayer_5], properties {weights [1,1,1]}}
 {id=WavePlayer_3, properties {soundWaveAssetPtr=/VividRazorExotics/Gameplay/SlapCannon/Sounds/Waves/StickyGrenadeLauncher_DropLand_01}}
 {id=WavePlayer_4, properties {soundWaveAssetPtr=/VividRazorExotics/Gameplay/SlapCannon/Sounds/Waves/StickyGrenadeLauncher_DropLand_02}}
 {id=WavePlayer_5, properties {soundWaveAssetPtr=/VividRazorExotics/Gameplay/SlapCannon/Sounds/Waves/StickyGrenadeLauncher_DropLand_03}}
]
properties {
 attenuationSettings=&Fort_Weapon_Drop_Att
 maxDistance=3000
 soundClassObject=&Fort_Loot_Collect
 volumeMultiplier=0.3
 concurrencyOverrides {resolutionRule=PreventNew}
 concurrencySet [&Weapon_Drop_StopOldest]
}
